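; Gameplay ability shared by Irwin NPCs that heals the activating pawn, presumably after eating, when triggered by a gameplay event.
; It treats the event magnitude as a percentage of the pawn's max health and passes the result as the SetByCaller.HealthRegen value on a configurable gameplay effect.
; It applies that effect to the pawn's own ability system, then ends immediately, leaving any regeneration behavior to the effect.
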(event Ability|EventActivateAbilityFromEvent
  (params (EventData_EventTag :type Gameplay%20Tag%20Structure) (EventData_Instigator :type Actor%20Object%20Reference) (EventData_Target :type Actor%20Object%20Reference) (EventData_OptionalObject :type Object%20Reference) (EventData_OptionalObject2 :type Object%20Reference) (EventData_ContextHandle :type Gameplay%20Effect%20Context%20Handle%20Structure) (EventData_InstigatorTags :type Gameplay%20Tag%20Container%20Structure) (EventData_TargetTags :type Gameplay%20Tag%20Container%20Structure) (EventData_EventMagnitude :type Float%20%28single-precision%29) (EventData_TargetData :type Gameplay%20Ability%20Target%20Data%20Handle%20Structure) (EventData_InstancedEventData :type Instanced%20Struct%20Structure))
  (bind _returnvalue (Ability|GetActivatingPawn))
  (bind _returnvalue_2 (Ability|GameplayEffect|AssignTagSetbyCallerMagnitude (Ability|MakeOutgoingGameplayEffectSpec (Variables|Default|GetGameplay%20Effect%20Class)) "(TagName=\"SetByCaller.HealthRegen\")" (* (* EventData_EventMagnitude 0.01) (FortPawn|GetMaxHealth _returnvalue))))
  (GameplayEffects|ApplyGameplayEffectSpecToSelf (Ability|GetAbilitySystemComponent _returnvalue) _returnvalue_2)
  (Ability|EndAbility))
