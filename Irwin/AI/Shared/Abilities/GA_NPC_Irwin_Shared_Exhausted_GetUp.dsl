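; Gameplay ability for the shared Irwin NPC set that, going by its name, plays the get-up from the Exhausted state.
; On activation it commits, then plays the configured montage through a play-and-wait-for-trigger task, or ends at once if no montage is valid.
; Only the montage's Completed output is wired: it removes the configured loose gameplay tags from the avatar actor and ends the ability.
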(event Ability|EventActivateAbility
  (bind _montage_to_play (Variables|Default|GetMontage%20to%20Play))
  (Ability|CommitAbility)
  (Utilities|IsValid _montage_to_play
    (:Is%20Valid
      (Ability|Tasks|PlayMontageandWaitforTrigger "None" _montage_to_play
        (:Completed
          (Ability|GameplayEffect|RemoveLooseGameplayTags (Ability|GetAvatarActorfromActorInfo) (Variables|Default|GetRemoved%20Gameplay%20Tags))
          (Ability|EndAbility))))
    (:Is%20Not%20Valid
      (Ability|EndAbility))))

(event Ability|EventOnEndAbility
  (params (bWasCancelled :type Boolean)))
