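; Shared Irwin NPC gameplay ability that plays an exhausted lay-down montage, lasting until the pawn takes damage.
; On activation it commits, starts the montage, binds to the AI pawn's OnDamaged and adds an optional looped gameplay cue, ending at once if no montage is valid.
; The montage trigger applies configurable loose gameplay tags to the avatar.
; Damage removes those tags, unbinds the delegate and ends the ability.
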(event Ability|EventActivateAbility
  (bind _montage_to_play (Variables|Default|GetMontage%20to%20Play))
  (bind _loopedgameplaycuetag (Variables|Default|GetLoopedGameplayCueTag))
  (|CallParentFunction)
  (Ability|CommitAbility)
  (Utilities|IsValid _montage_to_play
    (:Is%20Valid
      (Ability|Tasks|PlayMontageandWaitforTrigger "None" _montage_to_play
        (:then
          (bind _asfort_aipawn (Utilities|Casting|CastToFortAIPawn (Ability|GetAvatarActorfromActorInfo)))
          (EventDispatchers|BindEventtoOnDamaged _asfort_aipawn (delegate Custom|OnDamaged))
          ;[1 1 1] Optional Gameplay Cue
          (if (GameplayTags|IsGameplayTagValid _loopedgameplaycuetag)
            (Ability|AddGameplayCueToOwner _loopedgameplaycuetag)))
        (:Triggered
          (Ability|GameplayEffect|AddLooseGameplayTags (Ability|GetAvatarActorfromActorInfo) (Variables|Default|GetAdded%20Gameplay%20Tags)))))
    (:Is%20Not%20Valid
      (Ability|EndAbility))))

(event Custom|OnDamaged
  (params (DamagedActor :type Actor%20Object%20Reference) (Damage :type Float%20%28single-precision%29) (InstigatedBy :type Controller%20Object%20Reference) (DamageCauser :type Actor%20Object%20Reference) (HitLocation :type Vector) (FHitComponent :type Primitive%20Component%20Object%20Reference) (BoneName :type Name) (Momentum :type Vector))
  (Ability|GameplayEffect|RemoveLooseGameplayTags DamagedActor (Variables|Default|GetAdded%20Gameplay%20Tags))
  (EventDispatchers|UnbindEventfromOnDamaged (stale Ability|EventActivateAbility _asfort_aipawn) (delegate Custom|OnDamaged))
  (Ability|EndAbility))
